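; Blueprint interface for vehicle lock-on helpers in Athena drivable vehicles.
; It declares a single function, GetLockOnComponent, through which implementers expose their lock-on component to callers.
; The interface holds no logic of its own; the function stub returns an empty value.
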(fn GetLockOnComponent
  (return 0))
