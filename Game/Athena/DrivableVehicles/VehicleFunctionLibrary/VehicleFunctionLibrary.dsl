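; Vehicle function library for Fortnite drivable vehicles, with launch and rotate helpers, FX/layered-audio parameter setters that cache values, and an explosion routine.
; The server-side explosion kicks and spins the vehicle, then gathers line-of-sight overlaps plus all passengers.
; Targets get type-selected GameplayEffects with SetByCaller damage, and vehicles and players receive radial knockback with minimum lift.
; A second sphere trace damages building hits, ignoring prior targets.
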(fn LaunchVehicle
  (params (Vehicle :type Fort%20Athena%20Vehicle%20Object%20Reference) (LaunchVelocity :type Vector%20%28by%20ref%29))
  (Utilities|IsValid Vehicle
    (:Is%20Valid
      (Audio|BroadcastAbilityImpactatLocation 0 0 Vehicle LaunchVelocity 777.0 "None" true))
    (:Is%20Not%20Valid
      _))
  (return))

(fn RotateVehicle
  (params (Vehicle :type Fort%20Athena%20Vehicle%20Object%20Reference) (Pitch :type Float%20%28double-precision%29) (Yaw :type Float%20%28double-precision%29) (Roll :type Float%20%28double-precision%29))
  (Utilities|IsValid Vehicle
    (:Is%20Valid
      (Audio|BroadcastAbilitySpinThrustatLocation 0 0 Vehicle (Utilities|Operators|Add (* (Transformation|GetActorRightVector Vehicle) Pitch) (* (Transformation|GetActorUpVector Vehicle) Yaw) (* (Transformation|GetActorForwardVector Vehicle) Roll)) 777.0))
    (:Is%20Not%20Valid
      _))
  (return))

(fn SetFXComponentFloatParamFastUnsafe
  (params (FXSystemComponent :type FXSystem%20Component%20Object%20Reference) (ParamName :type Name) (CachedValueRef :type Float%20%28single-precision%29%20%28by%20ref%29) (NewValue :type Float%20%28single-precision%29))
  (if (!= CachedValueRef NewValue)
    (Variables|SetBy-RefVar CachedValueRef NewValue)
    (Effects|Components|ParticleSystem|SetFloatParameter FXSystemComponent ParamName NewValue)))

(fn SetLayeredAudioComponentFloatParamFastUnsafe
  (params (Layered%20Audio%20Component :type Fort%20Layered%20Audio%20Component%20Object%20Reference) (ParamName :type Name) (CachedValueRef :type Float%20%28single-precision%29%20%28by%20ref%29) (NewValue :type Float%20%28single-precision%29))
  (if (!= CachedValueRef NewValue)
    (Variables|SetBy-RefVar CachedValueRef NewValue)
    (Class|FortLayeredAudioComponent|SetFloatParameterExt Layered%20Audio%20Component ParamName NewValue)))

(fn SetLayeredAudioComponentFloatParamFastUnsafeNoGate
  (params (Layered%20Audio%20Component :type Fort%20Layered%20Audio%20Component%20Object%20Reference) (ParamName :type Name) (CachedValueRef :type Float%20%28single-precision%29%20%28by%20ref%29) (NewValue :type Float%20%28single-precision%29))
  (Variables|SetBy-RefVar CachedValueRef NewValue)
  (Class|FortLayeredAudioComponent|SetFloatParameterExt Layered%20Audio%20Component ParamName NewValue))

(fn VehicleExplodeAndDamageNearbyActors
  (params (Vehicle :type Fort%20Athena%20Vehicle%20Object%20Reference) (Instigator :type Actor%20Object%20Reference) (Radius :type Float%20%28double-precision%29) (GE%20Vehicle%20Damage :type Gameplay%20Effect%20Class%20Reference) (GE%20Building%20Damage :type Gameplay%20Effect%20Class%20Reference) (GE%20Player%20Damage :type Gameplay%20Effect%20Class%20Reference) (Vehicle%20Damage :type Float%20%28double-precision%29) (Player%20Damage :type Float%20%28double-precision%29) (Building%20Damage :type Float%20%28double-precision%29) (Ignore%20Line%20of%20Sight%20Non%20Building%20Actors :type Boolean) (Ignore%20Line%20of%20Sight%20Building%20Actors :type Boolean :default true) (Vehicle%20Knockback%20Magnitude :type Float%20%28double-precision%29) (Player%20Knockback%20Magnitude :type Float%20%28double-precision%29))
  (bind _instigator (Variables|GetInstigator))
  (bind _returnvalue (Utilities|IsValid _instigator :!class /Script/Engine.KismetSystemLibrary))
  (bind _vehicle (Variables|GetVehicle))
  (bind _returnvalue_2 (select _returnvalue _instigator _vehicle))
  (bind _returnvalue_3 (Ability|GetFortAbilitySystemComponent _returnvalue_2))
  (bind _instigator_2 (Variables|GetInstigator))
  (bind _returnvalue_4 (Utilities|IsValid _instigator_2 :!class /Script/Engine.KismetSystemLibrary))
  (bind _vehicle_2 (Variables|GetVehicle))
  (bind _returnvalue_5 (select _returnvalue_4 _instigator_2 _vehicle_2))
  (bind _returnvalue_6 (Ability|GetFortAbilitySystemComponent _returnvalue_5))
  (bind _targetedactors (LocalVariables|GetTargetedActors))
  ;[1 1 1] Collect Affected Bodies
  (if (Networking|IsServer 0)
    (Audio|BroadcastAbilityImpactatLocation 0 0 Vehicle (Math|Vector|MakeVector 0.0 0.0 3e+02) 777.0 "None" true)
    (Audio|BroadcastAbilitySpinThrustatLocation 0 0 Vehicle (* (Transformation|GetRightVector (Variables|/Script/FortniteGame.FortAthenaVehicle|GetSkeletalMesh Vehicle)) -5e+02) 777.0)
    (bind _outactors (Collision|SphereOverlapActors 0 (Transformation|GetActorLocation Vehicle) (Variables|GetRadius) (Utilities|Array|MakeArray "ObjectTypeQuery3" "ObjectTypeQuery1" "ObjectTypeQuery2" "ObjectTypeQuery5" "ObjectTypeQuery7") 0 (Utilities|Array|MakeArray Vehicle)))
    (for _array_element _outactors
      (bind _in_line_of_sight (Class|MissionBlueprintFunctionLibrary|CheckBasicActorLineOfSight Vehicle "0, 0,20.000000" _array_element "0, 0, 0" true "TraceTypeQuery1" "None" 15.0 true))
      (if _in_line_of_sight
        (Utilities|Array|AddUnique (LocalVariables|GetTargetedActors) _array_element)))
    ;;[1 1 1] Also do damage to players in vehicle
    (for _array_element_2 (Vehicle|GetAllPassengers Vehicle)
      (Utilities|Array|AddUnique (LocalVariables|GetTargetedActors) _array_element_2))
    ;[1 1 1] Apply Damage
    (for _array_element_3 (LocalVariables|GetTargetedActors)
      (if (Variables|/Script/Engine.Actor|GetbCanBeDamaged _array_element_3)
        (bind _returnvalue_7 (Ability|GameplayEffect|AssignTagSetbyCallerMagnitude (GameplayEffects|MakeOutgoingSpec :self _returnvalue_6 :GameplayEffectClass (select (.bSuccess (Utilities|Casting|CastToFortPlayerPawn _array_element_3)) (Variables|GetGE%20Player%20Damage) (select (.bSuccess (Utilities|Casting|CastToBuildingSMActor _array_element_3)) (Variables|GetGE%20Building%20Damage) (Variables|GetGE%20Vehicle%20Damage))) :Level 0.0) "(TagName=\"SetByCaller.Damage\")" (select (.bSuccess (Utilities|Casting|CastToFortPlayerPawn _array_element_3)) (Variables|GetPlayer%20Damage) (Variables|GetVehicle%20Damage))))
        (GameplayEffects|FortApplyGameplayEffectSpecToTarget :self _returnvalue_6 :GameplayEffectSpecHandle _returnvalue_7 :TargetOfEffect (Ability|GetFortAbilitySystemComponent _array_element_3) :OptionalSourceObject (Variables|GetVehicle) :OptionalEffectCauser (Variables|GetVehicle) :OptionalDamageSource (Variables|GetVehicle))
        (bind _asfort_athena_vehicle (Utilities|Casting|CastToFortAthenaVehicle _array_element_3)
          (:then
            (CallFunction|LaunchVehicle _asfort_athena_vehicle (Math|Vector|MakeVector (.x (* (Math|Vector|Normalize (- (Transformation|GetActorLocation _asfort_athena_vehicle) (Transformation|GetActorLocation (Variables|GetVehicle)))) (Variables|GetVehicle%20Knockback%20Magnitude))) (.y (* (Math|Vector|Normalize (- (Transformation|GetActorLocation _asfort_athena_vehicle) (Transformation|GetActorLocation (Variables|GetVehicle)))) (Variables|GetVehicle%20Knockback%20Magnitude))) (Math|Float|Max(Float) (+ (.z (* (Math|Vector|Normalize (- (Transformation|GetActorLocation _asfort_athena_vehicle) (Transformation|GetActorLocation (Variables|GetVehicle)))) (Variables|GetVehicle%20Knockback%20Magnitude))) 1.0) 5e+02))))
          (:CastFailed
            (if (.bSuccess (Utilities|Casting|CastToFortPlayerPawn _array_element_3))
              (Pawn|Character|LaunchCharacterJump (Utilities|Casting|CastToFortPlayerPawn _array_element_3) (Math|Vector|MakeVector (.x (* (Math|Vector|Normalize (- (Transformation|GetActorLocation _array_element_3) (Transformation|GetActorLocation (Variables|GetVehicle)))) (Variables|GetPlayer%20Knockback%20Magnitude))) (.y (* (Math|Vector|Normalize (- (Transformation|GetActorLocation _array_element_3) (Transformation|GetActorLocation (Variables|GetVehicle)))) (Variables|GetPlayer%20Knockback%20Magnitude))) (Math|Float|Max(Float) (+ (.z (* (Math|Vector|Normalize (- (Transformation|GetActorLocation _array_element_3) (Transformation|GetActorLocation (Variables|GetVehicle)))) (Variables|GetPlayer%20Knockback%20Magnitude))) 1.0) 3e+02)) false false false true))))))
    (bind _outhits (Collision|MultiSphereTraceForObjects 0 (Transformation|GetActorLocation (Variables|GetVehicle)) (+ (Transformation|GetActorLocation (Variables|GetVehicle)) "0, 0,1.000000") (Variables|GetRadius) (LocalVariables|GetObjectArray) false _targetedactors))
    (for _array_element_4 _outhits
      (Utilities|Casting|CastToBuildingActor (.HitActor (Collision|BreakHitResult _array_element_4)))
      (bind _returnvalue_8 (Ability|GameplayEffect|AssignTagSetbyCallerMagnitude (GameplayEffects|MakeOutgoingSpec :self _returnvalue_3 :GameplayEffectClass (Variables|GetGE%20Building%20Damage) :Level 1.0) "(TagName=\"SetByCaller.Damage\")" (Variables|GetBuilding%20Damage)))
      (GameplayEffects|FortApplyGameplayEffectSpecToTarget :self _returnvalue_3 :GameplayEffectSpecHandle _returnvalue_8 :TargetOfEffect (Ability|GetFortAbilitySystemComponent (.HitActor (Collision|BreakHitResult _array_element_4))) :OptionalSourceObject (Variables|GetVehicle) :OptionalEffectCauser (Variables|GetVehicle) :OptionalDamageSource (Variables|GetVehicle)))
    (return _targetedactors)))
